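# Sound cue for the Raptor BlackOps glider's open sound in Battle Royale.
# A fixed short delay node feeds a single wave player that plays the glider open wave.
# Playback uses the shared glider open/close attenuation and concurrency, a glider-open audio property sheet and the Fort_Gliders_Land sound class at reduced volume.
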SoundCue /BRCosmetics/Sounds/Gliders/RaptorBlackOps/Glider_RaptorBlackOps_Open_Cue
root=Delay_2
nodes [
 {id=Delay_2, children [WavePlayer_0], properties {delayMax=0.3, delayMin=0.3}}
 {id=WavePlayer_0, properties {soundWaveAssetPtr=&/Glider_RaptorBlackOps_Open}}
]
properties {
 attenuationSettings=/Game/Sounds/Attenuation/Fort_Player_Glider_OpenClose_Athena_Att
 audioPropertiesSheet=/BRCosmetics/Sounds/Settings/PropertySheets/Glider/APS_Cosmetic_Glider_Open
 maxDistance=10000
 soundClassObject=&Fort_Gliders_Land
 volumeMultiplier=0.6
 concurrencySet [&Athena_GliderOpenClose_Conc]
}
