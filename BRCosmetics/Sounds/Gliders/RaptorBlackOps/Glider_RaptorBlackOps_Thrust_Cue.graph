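# Sound cue for the thrust sound of the Raptor Black Ops glider cosmetic.
# A mixer layers three waves at different volumes: this glider's loop-in sample plus quieter thrust samples borrowed from the BoneWasp and Shuttle gliders.
# A modulator at the root feeds the output, which uses player glider open/close attenuation, a glider sound class and a thrust-specific concurrency set.
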SoundCue /BRCosmetics/Sounds/Gliders/RaptorBlackOps/Glider_RaptorBlackOps_Thrust_Cue
root=Modulator_1
nodes [
 {id=Mixer_0, children [WavePlayer_12, WavePlayer_10, WavePlayer_5], properties {inputVolume [1,0.3,0.5]}}
 {id=Modulator_1, children [Mixer_0]}
 {id=WavePlayer_10, properties {soundWaveAssetPtr=/BRCosmetics/Sounds/Gliders/BoneWasp/Glider_BoneWasp_Thrust_01}}
 {id=WavePlayer_12, properties {soundWaveAssetPtr=&/Glider_RaptorBlackOps_Loop_In_02}}
 {id=WavePlayer_5, properties {soundWaveAssetPtr=/BRCosmetics/Sounds/Gliders/SFX_Glider_Shuttle_ThrustStart}}
]
properties {
 attenuationSettings=/Game/Sounds/Attenuation/Fort_Player_Glider_OpenClose_Athena_Att
 audioPropertiesSheet=&APS_Cosmetic_Glider_Loop
 maxDistance=10000
 soundClassObject=&Fort_Gliders_Land
 volumeMultiplier=0.65
 concurrencySet [&Athena_Glider_Thrust_Conc]
}
